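# Two-sided lit material, named as a decal backdrop for the Barrier toy ability, that builds a grey striped surface with rock normal detail.
# World-aligned UVs feed both a scaled rock normal sample and a rotated, repeating SphereMask that blends two colors into stripes.
# Rounded bounding-box UVW selects a signed SizeBias vector per side, transformed from local space, apparently to resize the mesh; the default bias is zero.
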Material /Game/Abilities/Toys/Barrier/M_Decal-Backdrop
graph {
 properties {pixelDepthOffsetMode=PDOM_Legacy, twoSided=true, usedShadingModels=Default Lit}
 nodes [
  n1 = MaterialFunctionCall!(WorldScale=VectorParameter() {parameterName=WorldScale, defaultValue=<rgba 20 20 20 0>}.RGB, TransformToLocalSpace) {materialFunction=/Game/Items/Traps/Materials/MF_Creative_WorldAlignedUVs}
  n2 = Frac(None=Rotator(Coordinate=@n1.Result, Time=ScalarParameter() {defaultValue=9.3, parameterName=Time}))
  LinearInterpolate(
   A=VectorParameter() {parameterName=ColorLow, defaultValue=<rgba 0.1927 0.1927 0.1927 0>}.RGB
   B=VectorParameter() {parameterName=ColorHigh, defaultValue=<rgba 0.2969 0.2969 0.2969 0>}.RGB
   Alpha=SphereMask(
    A=MaterialFunctionCall(Float2=@n2) {materialFunction=&BreakOutFloat2Components}.R
    B=Constant() {r=0.5}
    Radius=ScalarParameter() {defaultValue=0.2, parameterName=StripeRadius}
    Hardness=ScalarParameter() {defaultValue=0.9, parameterName=StripeHarndess}
   )
  )
  Multiply(
   A=TextureSample(UVs=@n1.Result) {samplerType=SAMPLERTYPE_Normal, texture=/Game/Environments/Sets/Arid/Textures/RockDetail_normal}.RGB
   B=VectorParameter() {parameterName=NormalStrength, defaultValue=<rgba 0.3 0.3 1 0>}.RGB
  )
  n3 = VectorParameter() {parameterName=SizeBias, defaultValue=<rgba 0 0 0 1>}
  n4 = Multiply(A=@n3.R) {constB=-1}
  n5 = Multiply(A=@n3.G) {constB=-1}
  n6 = MaterialFunctionCall() {materialFunction=&BoundingBoxBased_0-1_UVW}
  n7 = Multiply(A=@n3.RGB, B=Round(None=@n6.B))
  Transform(None=LinearInterpolate(
    A=MaterialFunctionCall(X=@n4, Y=@n5, Z=@n7) {materialFunction=&MakeFloat3}.Result
    B=@n3.RGB
    Alpha=Round(None=@n6.RGB)
   )) {transformSourceType=TRANSFORMSOURCE_Local}
 ]
}
